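# Looping stereo ambient MetaSound for stage 01 of the GrillMoon portal prop.
# It passes its Sounds array, defaulting to the Stage_01 asset, and the play trigger into the shared MSS_PlayRandom_Loop graph, routing that graph's left and right outputs straight out.
# Playback uses portal-specific attenuation and loop concurrency under the global ambient sound class.
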MetaSoundSource /Hermes/Props/GrillMoon/Audio/MSS_GrillMoon_Portal_Stage_01
inputs [
 Sounds {type=WaveAsset:Array}
 UE.Source.OnPlay {type=Trigger}
 UE.Source.StartTime.StartTime {type=Time}
]
interfaces [UE.OutputFormat.Stereo:1.0, UE.Source.StartTime:1.0, UE.Source:1.1]
outputs [
 "UE.OutputFormat.Stereo.Audio:0" {type=Audio}
 "UE.OutputFormat.Stereo.Audio:1" {type=Audio}
]
pages [Default {
  input_defaults {UE.Source.OnPlay=null, UE.Source.StartTime.StartTime=null, Sounds [&/GrillMoon_Portal_Stage_01]}
  nodes [MSS_PlayRandom_Loop_0 = /Metasound/DefaultAssets/Sources/MSS_PlayRandom_Loop(Sounds=^Sounds, UE.Source.OnPlay=^UE.Source.OnPlay)]
  links [
   {dst="@UE.OutputFormat.Stereo.Audio:0", src=MSS_PlayRandom_Loop_0.UE.OutputFormat.Stereo.Audio:0}
   {dst="@UE.OutputFormat.Stereo.Audio:1", src=MSS_PlayRandom_Loop_0.UE.OutputFormat.Stereo.Audio:1}
  ]
 }]
properties {
 assetImportData=&self:AssetImportData
 attenuationSettings=&/ATT_GrillMoon_Portal_Amb
 bLooping=true
 duration=10000
 maxDistance=8500
 numChannels=2
 outputFormat=Stereo
 sampleRate=48000
 soundClassObject=/Game/Sounds/FortSoundClassesAndModes/Fort_Ambient_Global
 volume=0.5
 concurrencyOverrides {maxCount=1}
 concurrencySet [&/CONC_GrillMoon_Portal_Loop]
}
