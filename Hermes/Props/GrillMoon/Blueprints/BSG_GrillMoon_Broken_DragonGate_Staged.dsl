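; Broken GrillMoon Dragon Gate prop whose portal stage follows calendar events and which repels players and pickups.
; After match, playlist and calendar sync are ready, a server call checks three calendar events and activates the matching Data Layer, unloading the others.
; Portal_Stage1–3 events set each stage directly.
; On authority, impulse box overlap applies a knockback GameplayEffect to non-NPC pawns, or tosses pickups using look-at rotator components as the direction vector.
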(event EventBeginPlay
  ; Begin Play
  (Class|FortAsyncActionAuthorityMatchReady|MatchReadyAsync
    (:OnReady
      (Class|FortAsyncActionCurrentPlaylistReady|PlaylistDataReadyAsync
        (:OnReady
          (MCPContext|InitialCalendarSyncCompleted
            (:OnSyncComplete
              (CallFunction|%28Server%29SetCorrectStage))))))))

(event Custom|Portal_Stage1
  ; Set Stage with Calendar / Data Layers
  (bind _returnvalue (WorldPartition|GetDataLayerManager))
  (DataLayers|SetDataLayerRuntimeState _returnvalue "/Hermes/DataLayers/DL_GrillMoon_PortalStage_01.DL_GrillMoon_PortalStage_01" "Activated" false :!class /Script/Engine.DataLayerManager) ; stage 01 - ACTIVE
  (DataLayers|SetDataLayerRuntimeState _returnvalue "/Hermes/DataLayers/DL_GrillMoon_PortalStage_02.DL_GrillMoon_PortalStage_02" "Unloaded" false :!class /Script/Engine.DataLayerManager) ; stage 02
  (DataLayers|SetDataLayerRuntimeState _returnvalue "/Hermes/DataLayers/DL_GrillMoon_PortalStage_03.DL_GrillMoon_PortalStage_03" "Unloaded" false :!class /Script/Engine.DataLayerManager) ; stage 03
)

(event Custom|Portal_Stage2
  ; Set Stage with Calendar / Data Layers
  (bind _returnvalue (WorldPartition|GetDataLayerManager))
  (DataLayers|SetDataLayerRuntimeState _returnvalue "/Hermes/DataLayers/DL_GrillMoon_PortalStage_01.DL_GrillMoon_PortalStage_01" "Unloaded" false :!class /Script/Engine.DataLayerManager) ; stage 01
  (DataLayers|SetDataLayerRuntimeState _returnvalue "/Hermes/DataLayers/DL_GrillMoon_PortalStage_02.DL_GrillMoon_PortalStage_02" "Activated" false :!class /Script/Engine.DataLayerManager) ; stage 02 - ACTIVE
  (DataLayers|SetDataLayerRuntimeState _returnvalue "/Hermes/DataLayers/DL_GrillMoon_PortalStage_03.DL_GrillMoon_PortalStage_03" "Unloaded" false :!class /Script/Engine.DataLayerManager) ; stage 03
)

(event Custom|%28server%29%20Set%20Correct%20Stage
  ; Set Stage with Calendar / Data Layers
  (Utilities|FlowControl|Sequence
    (:then_0
      (bind _biseventactive (CalendarEvents|GetCalendarEventInformation "RipeHoneydew_Portal_01"))
      (if _biseventactive
        (bind _returnvalue (WorldPartition|GetDataLayerManager))
        (DataLayers|SetDataLayerRuntimeState _returnvalue "/Hermes/DataLayers/DL_GrillMoon_PortalStage_01.DL_GrillMoon_PortalStage_01" "Activated" false :!class /Script/Engine.DataLayerManager) ; stage 01 - ACTIVE
        (DataLayers|SetDataLayerRuntimeState _returnvalue "/Hermes/DataLayers/DL_GrillMoon_PortalStage_02.DL_GrillMoon_PortalStage_02" "Unloaded" false :!class /Script/Engine.DataLayerManager) ; stage 02
        (DataLayers|SetDataLayerRuntimeState _returnvalue "/Hermes/DataLayers/DL_GrillMoon_PortalStage_03.DL_GrillMoon_PortalStage_03" "Unloaded" false :!class /Script/Engine.DataLayerManager) ; stage 03
      ))
    (:then_1
      (bind _biseventactive_2 (CalendarEvents|GetCalendarEventInformation "RipeHoneydew_Portal_02"))
      (if _biseventactive_2
        (bind _returnvalue_2 (WorldPartition|GetDataLayerManager))
        (DataLayers|SetDataLayerRuntimeState _returnvalue_2 "/Hermes/DataLayers/DL_GrillMoon_PortalStage_01.DL_GrillMoon_PortalStage_01" "Unloaded" false :!class /Script/Engine.DataLayerManager) ; stage 01
        (DataLayers|SetDataLayerRuntimeState _returnvalue_2 "/Hermes/DataLayers/DL_GrillMoon_PortalStage_02.DL_GrillMoon_PortalStage_02" "Activated" false :!class /Script/Engine.DataLayerManager) ; stage 02 - ACTIVE
        (DataLayers|SetDataLayerRuntimeState _returnvalue_2 "/Hermes/DataLayers/DL_GrillMoon_PortalStage_03.DL_GrillMoon_PortalStage_03" "Unloaded" false :!class /Script/Engine.DataLayerManager) ; stage 03
      ))
    (:then_2
      (bind _biseventactive_3 (CalendarEvents|GetCalendarEventInformation "RipeHoneydew_Portal_03"))
      (if _biseventactive_3
        (bind _returnvalue_3 (WorldPartition|GetDataLayerManager))
        (DataLayers|SetDataLayerRuntimeState _returnvalue_3 "/Hermes/DataLayers/DL_GrillMoon_PortalStage_01.DL_GrillMoon_PortalStage_01" "Unloaded" false :!class /Script/Engine.DataLayerManager) ; stage 01
        (DataLayers|SetDataLayerRuntimeState _returnvalue_3 "/Hermes/DataLayers/DL_GrillMoon_PortalStage_02.DL_GrillMoon_PortalStage_02" "Unloaded" false :!class /Script/Engine.DataLayerManager) ; stage 02
        (DataLayers|SetDataLayerRuntimeState _returnvalue_3 "/Hermes/DataLayers/DL_GrillMoon_PortalStage_03.DL_GrillMoon_PortalStage_03" "Activated" false :!class /Script/Engine.DataLayerManager) ; stage 03 - ACTIVE
      ))))

(event Custom|Portal_Stage3
  ; Set Stage with Calendar / Data Layers
  (bind _returnvalue (WorldPartition|GetDataLayerManager))
  (DataLayers|SetDataLayerRuntimeState _returnvalue "/Hermes/DataLayers/DL_GrillMoon_PortalStage_01.DL_GrillMoon_PortalStage_01" "Unloaded" false :!class /Script/Engine.DataLayerManager) ; stage 01
  (DataLayers|SetDataLayerRuntimeState _returnvalue "/Hermes/DataLayers/DL_GrillMoon_PortalStage_02.DL_GrillMoon_PortalStage_02" "Unloaded" false :!class /Script/Engine.DataLayerManager) ; stage 02
  (DataLayers|SetDataLayerRuntimeState _returnvalue "/Hermes/DataLayers/DL_GrillMoon_PortalStage_03.DL_GrillMoon_PortalStage_03" "Activated" false :!class /Script/Engine.DataLayerManager) ; stage 03 - ACTIVE
)

(event OnComponentBeginOverlap(Box-ImpulseCollision)
  (params (OverlappedComponent :type Primitive%20Component%20Object%20Reference) (OtherActor :type Actor%20Object%20Reference) (OtherComp :type Primitive%20Component%20Object%20Reference) (OtherBodyIndex :type Integer) (bFromSweep :type Boolean) (SweepResult :type Hit%20Result%20Structure%20%28by%20ref%29))
  ; Knock Players Back
  (Networking|SwitchHasAuthority
    (:Authority
      (bind _asfort_player_pawn_athena (Utilities|Casting|CastToFortPlayerPawnAthena OtherActor)
        (:then
          (if (not (GameplayTags|HasMatchingGameplayTag _asfort_player_pawn_athena "(TagName=\"Pawn.Athena.NPC\")"))
            (GameplayEffects|FortApplyGameplayEffectToSelf :self (Ability|GetFortAbilitySystemComponent _asfort_player_pawn_athena) :GameplayEffectClass "/Hermes/Props/GrillMoon/Blueprints/GE_BrokenPortal_GrantKnockback.GE_BrokenPortal_GrantKnockback_C" :OptionalSourceObject self :OptionalEffectCauser self :Level 0.0)))
        (:CastFailed
          (bind _asfort_pickup (Utilities|Casting|CastToFortPickup OtherActor))
          (bind (_returnvalue_roll _returnvalue_pitch _returnvalue_yaw) (Math|Rotator|BreakRotator (Math|Rotator|FindLookatRotation (Transformation|GetActorLocation) (Transformation|GetActorLocation _asfort_pickup))))
          (Pickup|TossPickupinDirection _asfort_pickup (* (Math|Vector|MakeVector _returnvalue_roll _returnvalue_pitch (Math|Float|Absolute(Float) _returnvalue_yaw)) (ScalableFloat|GetValueatLevel (Variables|Default|GetPickupTossForce_Hf))) 0 false true))))))

(fn ConstructionScript)
